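# Looping outdoor daytime ambience cue for a Highlands biome point of interest, using Athena outdoor ambient attenuation and sound class at reduced volume.
# A quality-level node picks a looping quad wave for the first two levels or a stereo wave for the last two.
# Two nested, endlessly looping envelopers add variation: the inner runs a 30-second pitch curve, the outer a gentle 20-second volume dip.
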SoundCue /Game/Sounds/Helios/POI/Highlands/Biome_Highlands_01_Loop_Day_Cue
root=Enveloper_3
nodes [
 {
  id=Enveloper_0
  children [QualityLevel_0]
  properties {
   bLoop=true
   bLoopIndefinitely=true
   loopEnd=30
   pitchCurve {editorCurveData {keys [{interpMode=RCIM_Cubic, time=0}, <at 15.9468 0.9181 RCIM_Cubic>, {interpMode=RCIM_Cubic, time=30}]}}
   volumeCurve {editorCurveData {keys [<at 0 1>]}}
  }
 }
 {
  id=Enveloper_3
  children [Enveloper_0]
  properties {
   bLoop=true
   bLoopIndefinitely=true
   loopEnd=20
   pitchCurve {editorCurveData {keys [{time=0}]}}
   volumeCurve {editorCurveData {keys [
      {arriveTangent=-0.01, interpMode=RCIM_Cubic, leaveTangent=-0.01, time=0, value=1}
      <at 10 0.9 RCIM_Cubic>
      {arriveTangent=0.01, interpMode=RCIM_Cubic, leaveTangent=0.01, time=20, value=1}
     ]}}
  }
 }
 {id=QualityLevel_0, children [WavePlayer_2,WavePlayer_2,WavePlayer_3,WavePlayer_3]}
 {id=WavePlayer_2, properties {bLooping=true, soundWaveAssetPtr=&/Biome_Highlands_Day_01_Quad}}
 {id=WavePlayer_3, properties {bLooping=true, soundWaveAssetPtr=&/Biome_Highlands_Day_01_Stereo}}
]
properties {
 attenuationSettings=&Fort_Ambient_Athena_Att
 maxDistance=2097152
 soundClassObject=&Fort_Ambient_Outdoors_Athena
 volumeMultiplier=0.25
}
